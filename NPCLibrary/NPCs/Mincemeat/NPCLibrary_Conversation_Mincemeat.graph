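# Conversation database for the Mincemeat NPC, defining what it says when a player talks to it.
# A single entry point leads to one speech node with a default line and prioritized contextual lines.
# These react to a first-ever conversation or to specific player outfits, matched by character cosmetic ID.
# The speech node then links out to the shared NPC quest and basic services conversations.
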ConversationDatabase /NPCLibrary/NPCs/Mincemeat/NPCLibrary_Conversation_Mincemeat
entry_points [{tag=Conversation.Entry.Mincemeat, destinations [#1]}]
nodes [
 ConversationEntryPointNode_7 {
  id=#1
  class=ConversationEntryPointNode
  outputs [#2]
  properties {entryTag {tagName=Conversation.Entry.Mincemeat}}
 }
 Speech_C_4 {
  id=#2
  class=Speech_C
  outputs [#3, #4]
  properties {
   nodeName=""
   generalConfig {
    defaultMessage="MMMMMM!!!!!!"
    contextualMessages [
     {
      message="Mince... MEAT!!!"
      contextRequirements [{requirement {class=FortControllerRequirement_IsFirstEverConversationWithNPC}}]
      priority {curve {rowName=Tandem.Speech.Priority.FirstEverInteraction}}
     }
     {
      message="STRONG FISH PIE. TOUGH CRUST. YES!!"
      contextRequirements [{requirement {class=FortControllerRequirement_HasCID, properties {allowedCIds [/BRCosmetics/Athena/Items/Cosmetics/Characters/Character_BrawnyBass]}}}]
      priority {curve {rowName=Tandem.Speech.Priority.OutfitReaction}}
     }
     {
      message=Taco pie? Hmm... Yes...
      contextRequirements [{requirement {
         class=FortControllerRequirement_HasCID
         properties {allowedCIds [
           /BRCosmetics/Athena/Items/Cosmetics/Characters/CID_538_Athena_Commando_M_Taco
           /BRCosmetics/Athena/Items/Cosmetics/Characters/CID_797_Athena_Commando_F_Taco
          ]}
        }}]
      priority {curve {rowName=Tandem.Speech.Priority.OutfitReaction}}
     }
     {
      message=Chef... Make pie...
      contextRequirements [{requirement {class=FortControllerRequirement_HasCID, properties {allowedCIds [&Character_LocalZilla]}}}]
      priority {curve {rowName=Tandem.Speech.Priority.OutfitReaction}}
     }
     {
      message="BROOOTHER!!!!!!!!!!!"
      contextRequirements [{requirement {
         class=FortControllerRequirement_HasCID
         properties {allowedCIds [/BRCosmetics/Athena/Items/Cosmetics/Characters/CID_958_Athena_Commando_M_PieMan]}
        }}]
      priority {curve {rowName=Tandem.Speech.Priority.OutfitReaction}}
     }
     {
      message=Banana meat... So tender...
      contextRequirements [{requirement {
         class=FortControllerRequirement_HasCID
         properties {allowedCIds [&CID_349_Athena_Commando_M_Banana]}
        }}]
      priority {curve {rowName=Tandem.Speech.Priority.OutfitReaction}}
     }
    ]
   }
  }
 }
 ConversationLinkNode_0 {id=#4, class=ConversationLinkNode, properties {remoteEntryTag {tagName=Conversation.Entry.BasicServices}}}
 ConversationLinkNode_1 {id=#3, class=ConversationLinkNode, properties {remoteEntryTag {tagName=Conversation.Entry.NPCLibrary.Quests}}}
]
